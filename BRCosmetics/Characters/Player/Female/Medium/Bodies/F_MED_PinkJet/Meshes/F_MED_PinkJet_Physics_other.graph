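# Physics asset for the F_MED_PinkJet player body cosmetic, covering its dangling chain, waist bag with lock, strap and skirt pieces.
# Pelvis, spine_03 and thigh shapes form the body volumes, with bag and strap on limited pelvis joints and the lock hanging off the bag.
# Chain capsules link back to front between base spheres, free-swinging at the front.
# Front and back skirt boxes rotate freely on the pelvis; side boxes have no constraints.
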PhysicsAsset /BRCosmetics/Characters/Player/Female/Medium/Bodies/F_MED_PinkJet/Meshes/F_MED_PinkJet_Physics_other
bodies {
 dyn_chain_back_01_base_joint [{shapeType=Sphere, radius=1.4456, shapeName=dyn_chain_back_01_base_joint_sphere}]
 dyn_chain_back_01_joint [{
   shapeType=Capsule
   length=3.4862
   radius=1.3035
   shapeName=dyn_chain_back_01_joint_capsule
   center=<0 -0.1235 -1.2624>
   rotation=<rot 0.0001 -0.0004 5.0097>
  }]
 dyn_chain_front_01_base_joint [{shapeType=Sphere, radius=1.5646, shapeName=dyn_chain_front_01_base_joint_sphere}]
 dyn_chain_front_01_joint [{
   shapeType=Capsule
   length=2.8564
   radius=1.1582
   shapeName=dyn_chain_front_01_joint_capsule
   center=<1.6858 0 -1.4516>
   rotation=<rot 49.2698 -0.0002 -0.0003>
  }]
 dyn_chain_mid_03_joint [{
   shapeType=Capsule
   length=6.1975
   radius=1.3664
   shapeName=dyn_chain_mid_03_joint_capsule
   center=<-0.1583 0.4352 0.0166>
   rotation=<rot 82.0491 0.0001 0.0003>
  }]
 dyn_chain_mid_03_joint2 [{
   shapeType=Capsule
   length=5.534
   radius=1.2159
   shapeName=dyn_chain_mid_03_joint2_capsule
   center=<0.5085 -0.0079 0.1255>
   rotation=<rot -80.3464 161.4263 -179.9991>
  }]
 dyn_chain_mid_03_joint3 [{
   shapeType=Capsule
   length=4.8548
   radius=1.3307
   shapeName=dyn_chain_mid_03_joint3_capsule
   center=<-0.4234 -0.1019 0.5246>
   rotation=<rot 73.6204 5.5914 -6.1138>
  }]
 dyn_skirt_bk_trs_l [{
   shapeType=Box
   extentX=38.2692
   extentY=2.5876
   extentZ=15.4003
   shapeName=dyn_skirt_bk_a_l_box
   center=<-9.5721 0.2849 0.3641>
   rotation=<rot -13.0336 -16.7578 -18.9214>
  }]
 dyn_skirt_fr_trs_l [{
   shapeType=Box
   extentX=28.8697
   extentY=3.3559
   extentZ=19.2506
   shapeName=dyn_skirt_fr_a_l_box
   center=<-15.2564 0.1231 -0.4637>
   rotation=<rot -22.1265 4.3865 48.8984>
  }]
 dyn_skirt_side_b_r [{
   shapeType=Box
   extentX=18.5617
   extentY=7.5
   extentZ=7.844
   shapeName=dyn_skirt_side_b_r_box
   center=<0.0823 -4.808 9.5944>
   rotation=<rot 2.501 0 -15.0862>
  }]
 dyn_skirt_side_trs_l [{
   shapeType=Box
   extentX=37.5106
   extentY=2.9476
   extentZ=11.6406
   shapeName=dyn_skirt_side_a_l_box
   center=<9.2036 2.9833 -0.8231>
   rotation=<rot 0.0001 0 3.2552>
  }]
 dyn_strap [{
   shapeType=Box
   extentX=14.3633
   extentY=2.829
   extentZ=7.5
   shapeName=dyn_strap_box
   center=<14.8031 0.5104 0>
   rotation=<rot -0.0061 10.4023 0.0016>
  }]
 dyn_waist_bag [{
   shapeType=Box
   extentX=5.1229
   extentY=11.6271
   extentZ=7.5
   shapeName=dyn_waist_bag_box
   center=<2.4954 0.0679 -0.7654>
   rotation=<rot 0.0001 0 -2.5508>
  }]
 dyn_waist_bag_lock [{
   shapeType=Capsule
   length=2.1325
   radius=0.6422
   shapeName=dyn_waist_bag_lock_capsule
   center=<1.9821 -0.0026 -0.0291>
   rotation=<rot 0 -86.8631 -85.013>
  }]
 pelvis [
  {shapeType=Capsule, length=10.9778, radius=9.0759, center=<3.828 -0.7652 5.067>, rotation=<rot 73.0471 83.26 173.2569>}
  {shapeType=Capsule, length=10.4498, radius=5.8518, shapeName=pelvis_capsule, center=<9.4997 0 -2.0643>}
 ]
 spine_03 [{shapeType=Sphere, radius=15, shapeName=spine_03_sphere, center=<6.1112 3.9452 -14.5073>}]
 thigh_l [{
   shapeType=Capsule
   length=19.8124
   radius=8.1144
   shapeName=thigh_r_capsule
   center=<-7.3375 -0.451 1.0541>
   rotation=<rot -0.8517 85.2593 95.4852>
  }]
 thigh_r [{
   shapeType=Capsule
   length=16.9616
   radius=10.9651
   shapeName=thigh_r_capsule
   center=<15.0412 0.2004 -0.6827>
   rotation=<rot -0.7427 83.8238 -82.7914>
  }]
}
constraints [
 {bone1Name=dyn_waist_bag_lock, bone2Name=dyn_waist_bag, swing1LimitDegrees=30, swing2LimitDegrees=30, twistLimitDegrees=0.01}
 {bone1Name=dyn_waist_bag, bone2Name=pelvis, swing1LimitDegrees=30, swing2LimitDegrees=10, twistLimitDegrees=0.01}
 {
  bone1Name=thigh_r
  bone2Name=pelvis
  swing1LimitDegrees=45
  swing1Motion=Free
  swing2LimitDegrees=45
  swing2Motion=Free
  twistLimitDegrees=45
  twistMotion=Free
 }
 {bone1Name=dyn_strap, bone2Name=pelvis, swing1LimitDegrees=60, swing2LimitDegrees=40, twistLimitDegrees=15}
 {
  bone1Name=thigh_l
  bone2Name=pelvis
  swing1LimitDegrees=45
  swing1Motion=Free
  swing2LimitDegrees=45
  swing2Motion=Free
  twistLimitDegrees=45
  twistMotion=Free
 }
 {bone1Name=dyn_chain_mid_03_joint, bone2Name=dyn_chain_mid_03_joint2, swing1LimitDegrees=45, swing2LimitDegrees=45, twistLimitDegrees=10}
 {bone1Name=dyn_chain_mid_03_joint2, bone2Name=dyn_chain_back_01_joint, swing1LimitDegrees=45, swing2LimitDegrees=45, twistLimitDegrees=10}
 {bone1Name=dyn_chain_mid_03_joint3, bone2Name=dyn_chain_mid_03_joint, swing1LimitDegrees=45, swing2LimitDegrees=45, twistLimitDegrees=10}
 {
  bone1Name=dyn_chain_back_01_joint
  bone2Name=dyn_chain_back_01_base_joint
  swing1LimitDegrees=30
  swing2LimitDegrees=30
  twistLimitDegrees=10
 }
 {bone1Name=dyn_chain_front_01_joint, bone2Name=dyn_chain_mid_03_joint3, swing1LimitDegrees=45, swing2LimitDegrees=45, twistLimitDegrees=10}
 {
  bone1Name=dyn_chain_front_01_joint
  bone2Name=dyn_chain_front_01_base_joint
  swing1LimitDegrees=45
  swing1Motion=Free
  swing2LimitDegrees=45
  swing2Motion=Free
  twistLimitDegrees=10
  twistMotion=Free
 }
 {
  bone1Name=dyn_skirt_fr_trs_l
  bone2Name=pelvis
  swing1LimitDegrees=45
  swing1Motion=Free
  swing2LimitDegrees=45
  swing2Motion=Free
  twistLimitDegrees=45
  twistMotion=Free
 }
 {
  bone1Name=dyn_skirt_bk_trs_l
  bone2Name=pelvis
  swing1LimitDegrees=45
  swing1Motion=Free
  swing2LimitDegrees=45
  swing2Motion=Free
  twistLimitDegrees=45
  twistMotion=Free
 }
]
