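# Physics asset for the skirt of the F_MED_PinkJet female character body, covering front, side and back panels despite the skirt_front name.
# Six two-box chains, left and right per region, hang from small base capsules with tight angular limits, except a wider primary swing at the back.
# Pelvis, spine and thigh shapes supply body volumes, with an extra box on the left thigh.
# Base bones and thighs link to the pelvis with free angular motion.
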PhysicsAsset /BRCosmetics/Characters/Player/Female/Medium/Bodies/F_MED_PinkJet/Meshes/F_MED_PinkJet_Physics_skirt_front
bodies {
 dyn_skirt_bk_a_base [{shapeType=Capsule, length=0.505, radius=0.505, shapeName=dyn_skirt_bk_a_base_capsule, rotation=<rot 0 -90 0>}]
 dyn_skirt_bk_a_base_l [{shapeType=Capsule, length=0.505, radius=0.505, shapeName=dyn_skirt_bk_a_base_l_capsule, rotation=<rot 0 -90 0>}]
 dyn_skirt_bk_a_base_r [{shapeType=Capsule, length=0.505, radius=0.505, shapeName=dyn_skirt_bk_a_base_r_capsule, rotation=<rot 0 -90 0>}]
 dyn_skirt_bk_a_l [{
   shapeType=Box
   extentX=10.1069
   extentY=2.3091
   extentZ=16.5988
   shapeName=dyn_skirt_bk_a_l_box
   center=<-9.3493 -0.3157 -3.6578>
   rotation=<rot 8.5506 6.7559 38.5>
  }]
 dyn_skirt_bk_a_r [{
   shapeType=Box
   extentX=10.1069
   extentY=2.7129
   extentZ=15.6228
   shapeName=dyn_skirt_bk_a_l_box
   center=<9.4878 -0.1423 2.2688>
   rotation=<rot 8.7062 5.3106 -148.4528>
  }]
 dyn_skirt_bk_b_l [{
   shapeType=Box
   extentX=10.7299
   extentY=2.3858
   extentZ=16.2079
   shapeName=dyn_skirt_bk_b_l_box
   center=<-2.2852 -1.5361 -3.7113>
   rotation=<rot 8.5506 6.7559 38.5>
  }]
 dyn_skirt_bk_b_r [{
   shapeType=Box
   extentX=10.7299
   extentY=2.3352
   extentZ=15.2319
   shapeName=dyn_skirt_bk_b_l_box
   center=<2.8432 1.5262 2.4483>
   rotation=<rot 4.2589 12.2033 -146.8088>
  }]
 dyn_skirt_fr_a_base [{shapeType=Capsule, length=0.505, radius=0.6653, shapeName=dyn_skirt_fr_a_base_capsule, rotation=<rot 0 -90 0>}]
 dyn_skirt_fr_a_base_l [{shapeType=Capsule, length=0.2355, radius=1.0942, shapeName=dyn_skirt_fr_a_base_l_capsule, rotation=<rot 0 -90 0>}]
 dyn_skirt_fr_a_base_r [{shapeType=Capsule, length=0.505, radius=0.505, shapeName=dyn_skirt_fr_a_base_r_capsule, rotation=<rot 0 -90 0>}]
 dyn_skirt_fr_a_l [{
   shapeType=Box
   extentX=10.2551
   extentY=2.6139
   extentZ=15.1177
   shapeName=dyn_skirt_fr_a_l_box
   center=<-10.947 -0.7721 -1.0737>
   rotation=<rot 0.0001 0.0003 -4.1594>
  }]
 dyn_skirt_fr_a_r [{
   shapeType=Box
   extentX=10.2551
   extentY=2.6139
   extentZ=17.2797
   shapeName=dyn_skirt_fr_a_l_box
   center=<10.947 0 2.5759>
   rotation=<rot 4.1026 0.0007 -179.9999>
  }]
 dyn_skirt_fr_b_l [{
   shapeType=Box
   extentX=10.6962
   extentY=3.1643
   extentZ=15.0792
   shapeName=dyn_skirt_fr_b_l_box
   center=<-5.1357 0.898 -2.1895>
   rotation=<rot 0.0001 0.0003 -4.1594>
  }]
 dyn_skirt_fr_b_r [{
   shapeType=Box
   extentX=10.6734
   extentY=3.1416
   extentZ=16.2656
   shapeName=dyn_skirt_fr_b_l_box
   center=<5.0821 -1.0005 3.6033>
   rotation=<rot 4.1026 0.0007 -179.9999>
  }]
 dyn_skirt_side_a_base_l [{shapeType=Capsule, length=0.505, radius=0.505, shapeName=dyn_skirt_side_a_base_l_capsule, rotation=<rot 0 -90 0>}]
 dyn_skirt_side_a_base_r [{shapeType=Capsule, length=0.505, radius=0.505, shapeName=dyn_skirt_side_a_base_r_capsule, rotation=<rot 0 -90 0>}]
 dyn_skirt_side_a_l [{shapeType=Box, extentX=9.366, extentY=4.1134, extentZ=12.7586, shapeName=dyn_skirt_side_a_l_box, center=<11.4307 -1 0>, rotation=<rot 0 5 0>}]
 dyn_skirt_side_a_r [{
   shapeType=Box
   extentX=9.366
   extentY=1.7217
   extentZ=12.7586
   shapeName=dyn_skirt_side_a_l_box
   center=<10.1977 1.3253 1.2224>
   rotation=<rot 0.0004 -7.4477 -179.9999>
  }]
 dyn_skirt_side_b_l [{
   shapeType=Box
   extentX=11.0388
   extentY=4.3309
   extentZ=13.7281
   shapeName=dyn_skirt_side_b_l_box
   center=<2.5873 0.3 0>
   rotation=<rot 0 10.6725 0>
  }]
 dyn_skirt_side_b_r [{
   shapeType=Box
   extentX=11.0388
   extentY=3.0681
   extentZ=13.7281
   shapeName=dyn_skirt_side_b_l_box
   center=<2.3928 -0.1875 1.3653>
   rotation=<rot 0.0008 -12.6082 179.9997>
  }]
 pelvis [{shapeType=Capsule, length=15.2993, radius=7.5848, center=<-5.2316 -1.146 0>, rotation=<rot 0 -90 0>}]
 spine_01 [{shapeType=Capsule, length=20.4794, radius=7.7191, shapeName=spine_01_capsule, center=<-2.5766 -3.1453 0>}]
 thigh_l [
  {
   shapeType=Capsule
   length=27.7313
   radius=7.2686
   shapeName=thigh_l_capsule
   center=<-16.8998 0.6736 -0.2476>
   rotation=<rot 3.7075 -90.8125 -95.5759>
  }
  {
   shapeType=Box
   extentX=14.7234
   extentY=10.9813
   extentZ=4.7427
   shapeName=thigh_l_box
   center=<-29.7527 0.1007 -8.8784>
   rotation=<rot -5.0195 -0.0002 0>
  }
 ]
 thigh_r [{
   shapeType=Capsule
   length=29.4895
   radius=6.9974
   shapeName=thigh_l_capsule
   center=<16.9352 -0.1632 0.8274>
   rotation=<rot -0.1512 -91.2545 88.8615>
  }]
}
constraints [
 {
  bone1Name=dyn_skirt_bk_a_base
  bone2Name=pelvis
  swing1LimitDegrees=45
  swing1Motion=Free
  swing2LimitDegrees=45
  swing2Motion=Free
  twistLimitDegrees=45
  twistMotion=Free
 }
 {
  bone1Name=dyn_skirt_bk_a_base_l
  bone2Name=pelvis
  swing1LimitDegrees=45
  swing1Motion=Free
  swing2LimitDegrees=45
  swing2Motion=Free
  twistLimitDegrees=45
  twistMotion=Free
 }
 {
  bone1Name=dyn_skirt_bk_a_base_r
  bone2Name=pelvis
  swing1LimitDegrees=45
  swing1Motion=Free
  swing2LimitDegrees=45
  swing2Motion=Free
  twistLimitDegrees=45
  twistMotion=Free
 }
 {
  bone1Name=dyn_skirt_fr_a_base
  bone2Name=pelvis
  swing1LimitDegrees=45
  swing1Motion=Free
  swing2LimitDegrees=45
  swing2Motion=Free
  twistLimitDegrees=45
  twistMotion=Free
 }
 {
  bone1Name=dyn_skirt_fr_a_base_l
  bone2Name=pelvis
  swing1LimitDegrees=45
  swing1Motion=Free
  swing2LimitDegrees=45
  swing2Motion=Free
  twistLimitDegrees=45
  twistMotion=Free
 }
 {
  bone1Name=dyn_skirt_fr_a_base_r
  bone2Name=pelvis
  swing1LimitDegrees=45
  swing1Motion=Free
  swing2LimitDegrees=45
  swing2Motion=Free
  twistLimitDegrees=45
  twistMotion=Free
 }
 {
  bone1Name=dyn_skirt_side_a_base_l
  bone2Name=pelvis
  swing1LimitDegrees=45
  swing1Motion=Free
  swing2LimitDegrees=45
  swing2Motion=Free
  twistLimitDegrees=45
  twistMotion=Free
 }
 {
  bone1Name=dyn_skirt_side_a_base_r
  bone2Name=pelvis
  swing1LimitDegrees=45
  swing1Motion=Free
  swing2LimitDegrees=45
  swing2Motion=Free
  twistLimitDegrees=45
  twistMotion=Free
 }
 {
  bone1Name=thigh_l
  bone2Name=pelvis
  swing1LimitDegrees=45
  swing1Motion=Free
  swing2LimitDegrees=45
  swing2Motion=Free
  twistLimitDegrees=45
  twistMotion=Free
 }
 {
  bone1Name=thigh_r
  bone2Name=pelvis
  swing1LimitDegrees=45
  swing1Motion=Free
  swing2LimitDegrees=45
  swing2Motion=Free
  twistLimitDegrees=45
  twistMotion=Free
 }
 {bone1Name=dyn_skirt_fr_a_l, bone2Name=dyn_skirt_fr_a_base_l, swing1LimitDegrees=20, swing2LimitDegrees=5, twistLimitDegrees=10}
 {bone1Name=dyn_skirt_fr_b_l, bone2Name=dyn_skirt_fr_a_l, swing1LimitDegrees=20, swing2LimitDegrees=5, twistLimitDegrees=10}
 {bone1Name=dyn_skirt_fr_b_r, bone2Name=dyn_skirt_fr_a_r, swing1LimitDegrees=20, swing2LimitDegrees=5, twistLimitDegrees=10}
 {bone1Name=dyn_skirt_fr_a_r, bone2Name=dyn_skirt_fr_a_base_r, swing1LimitDegrees=20, swing2LimitDegrees=5, twistLimitDegrees=10}
 {bone1Name=dyn_skirt_side_a_l, bone2Name=dyn_skirt_side_a_base_l, swing1LimitDegrees=20, swing2LimitDegrees=3, twistLimitDegrees=5}
 {bone1Name=dyn_skirt_side_b_l, bone2Name=dyn_skirt_side_a_l, swing1LimitDegrees=20, swing2LimitDegrees=10, twistLimitDegrees=5}
 {bone1Name=dyn_skirt_side_b_r, bone2Name=dyn_skirt_side_a_r, swing1LimitDegrees=20, swing2LimitDegrees=20, twistLimitDegrees=5}
 {bone1Name=dyn_skirt_side_a_r, bone2Name=dyn_skirt_side_a_base_r, swing1LimitDegrees=20, swing2LimitDegrees=10, twistLimitDegrees=5}
 {bone1Name=dyn_skirt_bk_a_r, bone2Name=dyn_skirt_bk_a_base_r, swing1LimitDegrees=50, swing2LimitDegrees=10, twistLimitDegrees=5}
 {bone1Name=dyn_skirt_bk_b_r, bone2Name=dyn_skirt_bk_a_r, swing1LimitDegrees=50, swing2LimitDegrees=10, twistLimitDegrees=5}
 {bone1Name=dyn_skirt_bk_a_l, bone2Name=dyn_skirt_bk_a_base_l, swing1LimitDegrees=50, swing2LimitDegrees=10, twistLimitDegrees=5}
 {bone1Name=dyn_skirt_bk_b_l, bone2Name=dyn_skirt_bk_a_l, swing1LimitDegrees=50, swing2LimitDegrees=10, twistLimitDegrees=5}
]
