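# Post-copy animation Blueprint for the F_MED_PinkJet female body cosmetic, adding secondary motion to its skirt, strap and chains.
# The copied parent pose passes through bone constraints, two pelvis-space RigidBody simulations and many additive corrective bone transforms.
# Each update derives their alphas from movement state and direction, thigh and head rotations, and gravity override.
# Crouching and skydiving reduce skirt physics and swap chain physics for constraints.
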AnimBlueprint /BRCosmetics/Characters/Player/Female/Medium/Bodies/F_MED_PinkJet/Meshes/F_MED_PinkJet_AnimBP
target_skeleton=&/F_MED_PinkJet_Skeleton
graphs [
 AnimGraph {nodes [
   n1 = "Constraint - dyn_skirt_side_a_base_l (R)"(ComponentPose=Local To Component(LocalPose=Copy Pose From Mesh() {bUseAttachedParent=true}), Alpha=0.6, ConstraintWeights_0=1) {
    alphaBoolBlend {blendOption=Linear}
    boneToModify {boneName=dyn_skirt_side_a_base_l}
    constraintSetup [{transformType=Rotation, targetBone {boneName=thigh_l}}]
    constraintWeights [1]
   }
   n2 = "Constraint - dyn_skirt_side_a_base_r (R)"(ComponentPose=@n1, Alpha=0.9, ConstraintWeights_0=1) {
    alphaBoolBlend {blendOption=Linear}
    boneToModify {boneName=dyn_skirt_side_a_base_r}
    constraintSetup [{transformType=Rotation, targetBone {boneName=thigh_r}}]
    constraintWeights [1]
   }
   n3 = "Constraint - dyn_skirt_fr_a_base_l (TRTR)"!(ComponentPose=@n2, Alpha=1, ConstraintWeights_0=0, ConstraintWeights_1:float=1, ConstraintWeights_2:float=0, ConstraintWeights_3:float=0) -> (Pose) {
    alphaBoolBlend {blendOption=Linear}
    boneToModify {boneName=dyn_skirt_fr_a_base_l}
    constraintSetup [
     {transformType=Translation, targetBone {boneName=pelvis}}
     {transformType=Rotation, targetBone {boneName=thigh_l}}
     {transformType=Translation, targetBone {boneName=thigh_l}}
     {transformType=Rotation, targetBone {boneName=pelvis}}
    ]
    constraintWeights [1,1,1,1]
   }
   n4 = "Constraint - dyn_skirt_fr_a_base_r (PP)"!(ComponentPose=@n3, Alpha=1, ConstraintWeights_0=0.3, ConstraintWeights_1:float=0.7) -> (Pose) {
    alphaBoolBlend {blendOption=Linear}
    boneToModify {boneName=dyn_skirt_fr_a_base_r}
    constraintSetup [{targetBone {boneName=pelvis}}, {targetBone {boneName=thigh_r}}]
    constraintWeights [1,1]
   }
   n5 = RigidBody!(ComponentPose=@n4, Alpha=Get Dyn Skirt() -> (Dyn Skirt:double), OverrideWorldGravity:Vector=Get Gravity() -> (Gravity:Vector)) -> (Pose) {
    bOverrideWorldGravity=true
    overridePhysicsAsset=&/F_MED_PinkJet_Physics_skirt_front
    simulationSpace=BaseBoneSpace
    alphaBoolBlend {blendInTime=0.5, blendOption=Linear, blendOutTime=0.5}
    baseBoneRef {boneName=pelvis}
    componentAppliedLinearAccClamp=<3000 3000 3000>
    componentLinearAccScale=<0.25 0.25 0.25>
    componentLinearVelScale=<0.3 0.3 0.3>
    overrideWorldGravity {z=-980}
    simSpaceSettings {
     maxAngularAcceleration=50
     maxAngularVelocity=50
     maxLinearAcceleration=100
     maxLinearVelocity=100
     externalLinearDragV {x=20}
     externalLinearVelocity=<50 50 50>
    }
   }
   n6 = "Constraint - dyn_skirt_fr_a_base (PP)"!(ComponentPose=@n5, Alpha=1, ConstraintWeights_0=0.5, ConstraintWeights_1:float=0.5) -> (Pose) {
    alphaBoolBlend {blendOption=Linear}
    boneToModify {boneName=dyn_skirt_fr_a_base}
    constraintSetup [{targetBone {boneName=dyn_skirt_fr_a_l}}, {targetBone {boneName=dyn_skirt_fr_a_r}}]
    constraintWeights [1,1]
   }
   n7 = "Constraint - dyn_skirt_bk_a_base (PP)"!(ComponentPose=@n6, Alpha=1, ConstraintWeights_0=0.5, ConstraintWeights_1:float=0.5) -> (Pose) {
    alphaBoolBlend {blendOption=Linear}
    boneToModify {boneName=dyn_skirt_bk_a_base}
    constraintSetup [{targetBone {boneName=dyn_skirt_bk_a_l}}, {targetBone {boneName=dyn_skirt_bk_a_r}}]
    constraintWeights [1,1]
   }
   n8 = "Transform (Modify) Bone - Bone: dyn_skirt_fr_a"!(ComponentPose=@n7, bAlphaBoolEnabled:bool=Get bIsCrouching(), Translation="0,0,0", Rotation="-0.000,0,-10.116", Scale="1,1,1") -> (Pose) {
    alphaInputType=Bool
    rotationMode=BMM_Additive
    translationMode=BMM_Additive
    alphaBoolBlend {blendOption=Linear}
    boneToModify {boneName=dyn_skirt_fr_a}
    rotation {roll=-10.116}
   }
   n9 = "Transform (Modify) Bone - Bone: dyn_skirt_fr_a"(ComponentPose=@n8, bAlphaBoolEnabled=Get bIsSprinting(), Translation="0,0,0", Rotation="0.000,0.000,-9.100", Scale="1,1,1") {
    alphaInputType=Bool
    rotationMode=BMM_Additive
    translationMode=BMM_Additive
    alphaBoolBlend {blendOption=Linear}
    boneToModify {boneName=dyn_skirt_fr_a}
    rotation {roll=-9.1}
   }
   n10 = "Transform (Modify) Bone - Bone: dyn_skirt_bk_a_base_l"(
    ComponentPose=@n9
    Alpha=Get thigh_l_rot_alpha() -> (thigh_l_rot_alpha:double)
    Translation="0,0,0"
    Rotation="2.626,-15.945,-23.408"
    Scale="1,1,1"
   ) {
    rotationMode=BMM_Additive
    alphaBoolBlend {blendOption=Linear}
    boneToModify {boneName=dyn_skirt_bk_a_base_l}
    rotation=<rot 2.626 -15.945 -23.408>
   }
   n11 = "Transform (Modify) Bone - Bone: dyn_skirt_bk_a_base_r"(
    ComponentPose=@n10
    Alpha=Get thigh_r_rot_alpha() -> (thigh_r_rot_alpha:double)
    Translation="0,0,0"
    Rotation="-18.862,4.078,-11.082"
    Scale="1,1,1"
   ) {
    rotationMode=BMM_Additive
    alphaBoolBlend {blendOption=Linear}
    boneToModify {boneName=dyn_skirt_bk_a_base_r}
    rotation=<rot -18.862 4.078 -11.082>
   }
   n12 = "Transform (Modify) Bone - Bone: dyn_skirt_fr_a_l"!(
    ComponentPose=@n11
    bAlphaBoolEnabled:bool=Get bIsSprinting()
    Translation="1.845,0.000,0.627"
    Rotation="15.250,1.564,-2.630"
    Scale="1,1,1"
   ) -> (Pose) {
    alphaInputType=Bool
    rotationMode=BMM_Additive
    translationMode=BMM_Additive
    alphaBoolBlend {blendInTime=1, blendOption=Linear}
    boneToModify {boneName=dyn_skirt_fr_a_l}
    rotation=<rot 15.25 1.564 -2.63>
    translation {x=1.845, z=0.627}
   }
   n13 = "Transform (Modify) Bone - Bone: dyn_skirt_fr_a_base_l"(
    ComponentPose=@n12
    Alpha="Get Crouching Move Float "() -> ("Crouching Move Float ":float)
    Translation="2.860,-0.102,-0.147"
    Rotation="2,15.092,-0.932"
    Scale="1,1,1"
   ) {
    rotationMode=BMM_Additive
    rotationSpace=BCS_BoneSpace
    translationMode=BMM_Additive
    translationSpace=BCS_BoneSpace
    alphaBoolBlend {blendOption=Linear}
    alphaScaleBiasClamp {bInterpResult=true}
    boneToModify {boneName=dyn_skirt_fr_a_base_l}
    rotation=<rot 2 15.092 -0.932>
    translation=<2.86 -0.102 -0.147>
   }
   n14 = Get isCrouchingFloat() -> (isCrouchingFloat:double)
   n15 = "Transform (Modify) Bone - Bone: dyn_skirt_fr_a_base_l"(ComponentPose=@n13, Alpha=@n14, Translation="0,0,0", Rotation="-5.560,-11.847,-7.835", Scale="0.598,1,1.000") {
    rotationMode=BMM_Additive
    rotationSpace=BCS_BoneSpace
    scaleMode=BMM_Additive
    scaleSpace=BCS_BoneSpace
    translationMode=BMM_Additive
    translationSpace=BCS_BoneSpace
    alphaBoolBlend {blendOption=Linear}
    alphaScaleBiasClamp {bInterpResult=true}
    boneToModify {boneName=dyn_skirt_fr_a_base_l}
    rotation=<rot -5.56 -11.847 -7.835>
    scale {x=0.598}
   }
   n16 = "Transform (Modify) Bone - Bone: dyn_skirt_side_a_l"(ComponentPose=@n15, Alpha=@n14, Translation="0,0,0", Rotation="-5.548,-13.837,-12.336", Scale="1,1,1") {
    rotationMode=BMM_Additive
    rotationSpace=BCS_BoneSpace
    scaleMode=BMM_Additive
    scaleSpace=BCS_BoneSpace
    translationMode=BMM_Additive
    translationSpace=BCS_BoneSpace
    alphaBoolBlend {blendOption=Linear}
    alphaScaleBiasClamp {bInterpResult=true}
    boneToModify {boneName=dyn_skirt_side_a_l}
    rotation=<rot -5.548 -13.837 -12.336>
   }
   n17 = "Transform (Modify) Bone - Bone: dyn_skirt_fr_a_base"(ComponentPose=@n16, Alpha=@n14, Translation="0,0,0", Rotation="0,0,0", Scale="0.85,1,1") {
    rotationMode=BMM_Additive
    rotationSpace=BCS_BoneSpace
    scaleMode=BMM_Additive
    scaleSpace=BCS_BoneSpace
    alphaBoolBlend {blendOption=Linear}
    alphaScaleBiasClamp {bInterpResult=true}
    boneToModify {boneName=dyn_skirt_fr_a_base}
    scale {x=0.85}
   }
   n18 = Get R leg skydive fix() -> (R leg skydive fix:bool)
   n19 = "Transform (Modify) Bone - Bone: dyn_skirt_fr_a_base_r"!(ComponentPose=@n17, bAlphaBoolEnabled:bool=@n18, Translation="0,0,0", Rotation="-0.154,-11.039,0.217", Scale="1,1,1") -> (Pose) {
    alphaInputType=Bool
    rotationMode=BMM_Additive
    rotationSpace=BCS_BoneSpace
    alphaBoolBlend {blendOption=Linear}
    boneToModify {boneName=dyn_skirt_fr_a_base_r}
    rotation=<rot -0.154 -11.039 0.217>
   }
   n20 = "Transform (Modify) Bone - Bone: dyn_skirt_fr_a_r"!(ComponentPose=@n19, bAlphaBoolEnabled:bool=@n18, Translation="0,0,5", Rotation="-3.331,22.473,0.023", Scale="1,1,1") -> (Pose) {
    alphaInputType=Bool
    rotationMode=BMM_Additive
    rotationSpace=BCS_BoneSpace
    translationMode=BMM_Additive
    translationSpace=BCS_BoneSpace
    alphaBoolBlend {blendOption=Linear}
    boneToModify {boneName=dyn_skirt_fr_a_r}
    rotation=<rot -3.331 22.473 0.023>
    translation {z=5}
   }
   n21 = "Transform (Modify) Bone - Bone: dyn_skirt_side_a_base_l"!(ComponentPose=@n20, bAlphaBoolEnabled:bool=Get bIsCrouching(), Translation="0,0,0", Rotation="21.128,-3.612,14.318", Scale="0.6,1,1") -> (Pose) {
    alphaInputType=Bool
    rotationMode=BMM_Additive
    rotationSpace=BCS_BoneSpace
    scaleMode=BMM_Additive
    scaleSpace=BCS_BoneSpace
    alphaBoolBlend {blendInTime=0.2, blendOption=Linear, blendOutTime=0.2}
    boneToModify {boneName=dyn_skirt_side_a_base_l}
    rotation=<rot 21.128 -3.612 14.318>
    scale {x=0.6}
   }
   n22 = "Transform (Modify) Bone - Bone: dyn_skirt_side_a_base_r"!(
    ComponentPose=@n21
    bAlphaBoolEnabled:bool=Get Crouching Move() -> (Crouching Move:bool)
    Translation="0,0,0"
    Rotation="9.807,11.452,16.892"
    Scale="1,1,1"
   ) -> (Pose) {
    alphaInputType=Bool
    rotationMode=BMM_Additive
    rotationSpace=BCS_BoneSpace
    scaleMode=BMM_Additive
    scaleSpace=BCS_BoneSpace
    alphaBoolBlend {blendOption=Linear}
    boneToModify {boneName=dyn_skirt_side_a_base_r}
    rotation=<rot 9.807 11.452 16.892>
   }
   n23 = Get bIsSprinting()
   n24 = "Transform (Modify) Bone - Bone: dyn_skirt_bk_a_base_l"!(ComponentPose=@n22, bAlphaBoolEnabled:bool=@n23, Translation="0,0,0", Rotation="0,-5,0", Scale="1,1,1") -> (Pose) {
    alphaInputType=Bool
    rotationMode=BMM_Additive
    rotationSpace=BCS_BoneSpace
    alphaBoolBlend {blendInTime=0.2, blendOption=Linear, blendOutTime=0.2}
    boneToModify {boneName=dyn_skirt_bk_a_base_l}
    rotation {yaw=-5}
   }
   n25 = "Transform (Modify) Bone - Bone: dyn_skirt_bk_a_base"!(ComponentPose=@n24, bAlphaBoolEnabled:bool=@n23, Translation="0,0,0", Rotation="0,-5,0", Scale="1,1,1") -> (Pose) {
    alphaInputType=Bool
    rotationMode=BMM_Additive
    rotationSpace=BCS_BoneSpace
    alphaBoolBlend {blendInTime=0.2, blendOption=Linear, blendOutTime=0.2}
    boneToModify {boneName=dyn_skirt_bk_a_base}
    rotation {yaw=-5}
   }
   n26 = "Transform (Modify) Bone - Bone: dyn_skirt_bk_a_base_r"!(ComponentPose=@n25, bAlphaBoolEnabled:bool=@n23, Translation="0,0,0", Rotation="0,-5,0", Scale="1,1,1") -> (Pose) {
    alphaInputType=Bool
    rotationMode=BMM_Additive
    rotationSpace=BCS_BoneSpace
    alphaBoolBlend {blendInTime=0.2, blendOption=Linear, blendOutTime=0.2}
    boneToModify {boneName=dyn_skirt_bk_a_base_r}
    rotation {yaw=-5}
   }
   n27 = "Constraint - dyn_strap_base (PPPPPP)"!(
    ComponentPose=@n26
    Alpha=0.9
    ConstraintWeights_0=0.1003
    ConstraintWeights_1:float=0.003
    ConstraintWeights_2:float=0.0869
    ConstraintWeights_3:float=0.1185
    ConstraintWeights_4:float=0.6913
    ConstraintWeights_5:float=0
   ) -> (Pose) {
    alphaBoolBlend {blendOption=Linear}
    boneToModify {boneName=dyn_strap_base}
    constraintSetup [
     {targetBone {boneName=dyn_skirt_fr_b_r}}
     {targetBone {boneName=dyn_skirt_bk_b_r}}
     {targetBone {boneName=dyn_skirt_fr_a_r}}
     {targetBone {boneName=dyn_skirt_side_a_r}}
     {targetBone {boneName=dyn_skirt_side_b_r}}
     {targetBone {boneName=spine_01}}
    ]
    constraintWeights [1,1,1,1,1,1]
   }
   n28 = Get Con Chain() -> (Con Chain:double)
   n29 = "Constraint - dyn_chain_front_01_joint (PPP)"!(ComponentPose=@n27, Alpha=@n28, ConstraintWeights_0=0.6858, ConstraintWeights_1:float=0.3067, ConstraintWeights_2:float=0.0075) -> (Pose) {
    alphaBoolBlend {blendOption=Linear}
    boneToModify {boneName=dyn_chain_front_01_joint}
    constraintSetup [{targetBone {boneName=pelvis}}, {targetBone {boneName=dyn_skirt_fr_a_l}}, {targetBone {boneName=spine_01}}]
    constraintWeights [1,1,1]
   }
   n30 = "Constraint - dyn_chain_mid_03_joint3 (PPPPP)"!(
    ComponentPose=@n29
    Alpha=Get Con chain crounch() -> (Con chain crounch:double)
    ConstraintWeights_0=0.7132
    ConstraintWeights_1:float=0
    ConstraintWeights_2:float=0.2466
    ConstraintWeights_3:float=0.0403
    ConstraintWeights_4:float=0
   ) -> (Pose) {
    alphaBoolBlend {blendOption=Linear}
    boneToModify {boneName=dyn_chain_mid_03_joint3}
    constraintSetup [
     {targetBone {boneName=dyn_skirt_side_a_l}}
     {targetBone {boneName=dyn_skirt_side_b_l}}
     {targetBone {boneName=pelvis}}
     {targetBone {boneName=dyn_skirt_fr_a_l}}
     {targetBone {boneName=dyn_skirt_bk_b_l}}
    ]
    constraintWeights [1,1,1,1,1]
   }
   n31 = "Constraint - dyn_chain_mid_03_joint2 (PPPPPPPP)"!(
    ComponentPose=@n30
    Alpha=@n28
    ConstraintWeights_0=0.5123
    ConstraintWeights_1:float=0
    ConstraintWeights_2:float=0.4181
    ConstraintWeights_3:float=0.0306
    ConstraintWeights_4:float=0.0223
    ConstraintWeights_5:float=0.0167
    ConstraintWeights_6:float=0
    ConstraintWeights_7:float=0
   ) -> (Pose) {
    alphaBoolBlend {blendOption=Linear}
    boneToModify {boneName=dyn_chain_mid_03_joint2}
    constraintSetup [
     {targetBone {boneName=dyn_skirt_side_a_l}}
     {targetBone {boneName=dyn_skirt_side_b_l}}
     {targetBone {boneName=pelvis}}
     {targetBone {boneName=extra_back_pelvis_m}}
     {targetBone {boneName=dyn_skirt_bk_a_l}}
     {targetBone {boneName=dyn_skirt_fr_a_l}}
     {targetBone {boneName=dyn_skirt_fr_b_l}}
     {targetBone {boneName=dyn_skirt_bk_b_l}}
    ]
    constraintWeights [1,1,1,1,1,1,1,1]
   }
   n32 = "Constraint - dyn_chain_mid_03_joint2 (PPPPP)"!(
    ComponentPose=@n31
    Alpha=@n28
    ConstraintWeights_0=0.0847
    ConstraintWeights_1:float=0.1836
    ConstraintWeights_2:float=0.0073
    ConstraintWeights_3:float=0.718
    ConstraintWeights_4:float=0.0064
   ) -> (Pose) {
    alphaBoolBlend {blendOption=Linear}
    boneToModify {boneName=dyn_chain_mid_03_joint2}
    constraintSetup [
     {targetBone {boneName=dyn_skirt_side_a_l}}
     {targetBone {boneName=pelvis}}
     {targetBone {boneName=dyn_skirt_bk_a}}
     {targetBone {boneName=dyn_skirt_bk_a_l}}
     {targetBone {boneName=dyn_skirt_bk_b_l}}
    ]
    constraintWeights [1,1,1,1,1]
   }
   n33 = "Transform (Modify) Bone - Bone: dyn_chain_front_01_joint"!(ComponentPose=@n32, bAlphaBoolEnabled:bool=True, Translation="2.918,2.202,1.039", Rotation="0,0,0", Scale="1,1,1") -> (Pose) {
    alphaInputType=Bool
    rotationMode=BMM_Additive
    rotationSpace=BCS_BoneSpace
    translationMode=BMM_Additive
    translationSpace=BCS_BoneSpace
    alphaBoolBlend {blendOption=Linear}
    boneToModify {boneName=dyn_chain_front_01_joint}
    translation=<2.918 2.202 1.039>
   }
   n34 = RigidBody(ComponentPose=@n33, Alpha=Get Dyn Chain() -> (Dyn Chain:double), OverrideWorldGravity=Get Gravity()) {
    bOverrideWorldGravity=true
    overridePhysicsAsset=&/F_MED_PinkJet_Physics_other
    simulationSpace=BaseBoneSpace
    alphaBoolBlend {blendInTime=0.5, blendOption=Linear, blendOutTime=0.5}
    baseBoneRef {boneName=pelvis}
    componentAppliedLinearAccClamp=<2500 2500 2500>
    componentLinearAccScale=<0.35 0.35 0.35>
    componentLinearVelScale=<0.35 0.35 0.35>
    overrideWorldGravity {z=-980}
    simSpaceSettings {maxAngularAcceleration=3000, maxAngularVelocity=3000, maxLinearAcceleration=3000, maxLinearVelocity=3000}
   }
   n35 = Get isCrouchingFloat()
   n36 = "Transform (Modify) Bone - Bone: dyn_chain_mid_03_joint3"(ComponentPose=@n34, Alpha=@n35, Translation="2.669,1.097,-1.206", Rotation="-26.552,-1.996,0.771", Scale="1,1,1") {
    rotationMode=BMM_Additive
    rotationSpace=BCS_BoneSpace
    translationMode=BMM_Additive
    translationSpace=BCS_BoneSpace
    alphaBoolBlend {blendOption=Linear}
    alphaScaleBiasClamp {bInterpResult=true}
    boneToModify {boneName=dyn_chain_mid_03_joint3}
    rotation=<rot -26.552 -1.996 0.771>
    translation=<2.669 1.097 -1.206>
   }
   n37 = "Transform (Modify) Bone - Bone: dyn_chain_front_01_joint"!(ComponentPose=@n36, Alpha:float=@n35, Translation="-4.654,-1.910,-0.290", Rotation="41.211,-18.429,19.593", Scale="1,1,1") -> (Pose) {
    rotationMode=BMM_Additive
    rotationSpace=BCS_BoneSpace
    translationMode=BMM_Additive
    translationSpace=BCS_BoneSpace
    alphaBoolBlend {blendOption=Linear}
    alphaScaleBiasClamp {bInterpResult=true}
    boneToModify {boneName=dyn_chain_front_01_joint}
    rotation=<rot 41.211 -18.429 19.593>
    translation=<-4.654 -1.91 -0.29>
   }
   n38 = "Transform (Modify) Bone - Bone: dyn_chain_front_01_base_joint"(ComponentPose=@n37, Alpha=@n35, Translation="-0.452,0.461,-0.199", Rotation="0,0,0", Scale="1,1,1") {
    rotationMode=BMM_Additive
    rotationSpace=BCS_BoneSpace
    translationMode=BMM_Additive
    translationSpace=BCS_BoneSpace
    alphaBoolBlend {blendOption=Linear}
    alphaScaleBiasClamp {bInterpResult=true}
    boneToModify {boneName=dyn_chain_front_01_base_joint}
    translation=<-0.452 0.461 -0.199>
   }
   n39 = "Transform (Modify) Bone - Bone: dyn_strap_base"(
    ComponentPose=@n38
    Alpha=Get Left MoveFloat() -> (Left MoveFloat:double)
    Translation="0,0,0"
    Rotation="42.382,-26.305,-34.422"
    Scale="1,1,1"
   ) {
    rotationMode=BMM_Additive
    rotationSpace=BCS_BoneSpace
    alphaBoolBlend {blendInTime=0.2, blendOption=Linear, blendOutTime=0.2}
    alphaScaleBiasClamp {bInterpResult=true}
    boneToModify {boneName=dyn_strap_base}
    rotation=<rot 42.382 -26.305 -34.422>
   }
   n40 = "Transform (Modify) Bone - Bone: dyn_strap_base"(
    ComponentPose=@n39
    Alpha=Get Right MoveFloat() -> (Right MoveFloat:double)
    Translation="-0.677,1.283,0.506"
    Rotation="26.315,-24.694,-30.886"
    Scale="1,1,1"
   ) {
    rotationMode=BMM_Additive
    rotationSpace=BCS_BoneSpace
    translationMode=BMM_Additive
    translationSpace=BCS_BoneSpace
    alphaBoolBlend {blendInTime=0.5, blendOption=Linear, blendOutTime=0.5}
    alphaScaleBiasClamp {bInterpResult=true}
    boneToModify {boneName=dyn_strap_base}
    rotation=<rot 26.315 -24.694 -30.886>
    translation=<-0.677 1.283 0.506>
   }
   n41 = "Transform (Modify) Bone - Bone: dyn_strap_base"(
    ComponentPose=@n40
    Alpha=Get isCrouchSprinting() -> (isCrouchSprinting:float)
    Translation="-0.677,1.283,0.506"
    Rotation="26.315,-24.694,-30.886"
    Scale="1,1,1"
   ) {
    rotationMode=BMM_Additive
    rotationSpace=BCS_BoneSpace
    translationMode=BMM_Additive
    translationSpace=BCS_BoneSpace
    alphaBoolBlend {blendInTime=0.5, blendOption=Linear, blendOutTime=0.5}
    alphaScaleBiasClamp {bInterpResult=true}
    boneToModify {boneName=dyn_strap_base}
    rotation=<rot 26.315 -24.694 -30.886>
    translation=<-0.677 1.283 0.506>
   }
   n42 = "Transform (Modify) Bone - Bone: dyn_strap_base"!(
    ComponentPose=@n41
    bAlphaBoolEnabled:bool=Get back crouch strap() -> (back crouch strap:bool)
    Translation="-0.677,1.283,0.506"
    Rotation="56.024,-38.049,-47.243"
    Scale="1,1,1"
   ) -> (Pose) {
    alphaInputType=Bool
    rotationMode=BMM_Additive
    rotationSpace=BCS_BoneSpace
    translationMode=BMM_Additive
    translationSpace=BCS_BoneSpace
    alphaBoolBlend {blendInTime=0.5, blendOption=Linear, blendOutTime=0.5}
    boneToModify {boneName=dyn_strap_base}
    rotation=<rot 56.024 -38.049 -47.243>
    translation=<-0.677 1.283 0.506>
   }
   n43 = "Constraint - dyn_chain_front_01_base_joint (TPPPP)"!(
    ComponentPose=@n42
    Alpha=1
    ConstraintWeights_0=0.0007
    ConstraintWeights_1:float=0.6596
    ConstraintWeights_2:float=0.2235
    ConstraintWeights_3:float=0.1129
    ConstraintWeights_4:float=0.0033
   ) -> (Pose) {
    alphaBoolBlend {blendOption=Linear}
    boneToModify {boneName=dyn_chain_front_01_base_joint}
    constraintSetup [
     {transformType=Translation, targetBone {boneName=spine_03}}
     {targetBone {boneName=pelvis}}
     {targetBone {boneName=spine_02}}
     {targetBone {boneName=spine_01}}
     {targetBone {boneName=spine_04}}
    ]
    constraintWeights [1,1,1,1,1]
   }
   n44 = "Transform (Modify) Bone - Bone: dyn_chain_front_01_joint"(
    ComponentPose=@n43
    Alpha=Get Con End Chain() -> (Con End Chain:double)
    Translation="-1.545,0.651,0.930"
    Rotation="24.002,10.174,-33.331"
    Scale="1,1,1"
   ) {
    rotationMode=BMM_Additive
    rotationSpace=BCS_BoneSpace
    translationMode=BMM_Additive
    translationSpace=BCS_BoneSpace
    alphaBoolBlend {blendOption=Linear}
    boneToModify {boneName=dyn_chain_front_01_joint}
    rotation=<rot 24.002 10.174 -33.331>
    translation=<-1.545 0.651 0.93>
   }
   n45 = Get bIsJumping()
   n46 = "Transform (Modify) Bone - Bone: dyn_chain_back_01_joint"!(ComponentPose=@n44, bAlphaBoolEnabled:bool=@n45, Translation="-0.167,0.568,-0.553", Rotation="-0.013,-9.287,0.006", Scale="1,1,1") -> (Pose) {
    alphaInputType=Bool
    rotationMode=BMM_Additive
    rotationSpace=BCS_BoneSpace
    translationMode=BMM_Additive
    translationSpace=BCS_ParentBoneSpace
    alphaBoolBlend {blendInTime=0.4, blendOption=Linear, blendOutTime=0.4}
    boneToModify {boneName=dyn_chain_back_01_joint}
    rotation=<rot -0.013 -9.287 0.006>
    translation=<-0.167 0.568 -0.553>
   }
   n47 = "Transform (Modify) Bone - Bone: dyn_chain_front_01_joint"!(ComponentPose=@n46, bAlphaBoolEnabled:bool=@n45, Translation="0.014,1.002,0.307", Rotation="-17.313,0.058,0.045", Scale="1,1,1") -> (Pose) {
    alphaInputType=Bool
    rotationMode=BMM_Additive
    rotationSpace=BCS_BoneSpace
    translationMode=BMM_Additive
    translationSpace=BCS_ParentBoneSpace
    alphaBoolBlend {blendInTime=0.4, blendOption=Linear, blendOutTime=0.4}
    boneToModify {boneName=dyn_chain_front_01_joint}
    rotation=<rot -17.313 0.058 0.045>
    translation=<0.014 1.002 0.307>
   }
   n48 = "Transform (Modify) Bone - Bone: dyn_chain_mid_03_joint2"!(ComponentPose=@n47, bAlphaBoolEnabled:bool=Get bIsMoving2D(), Translation="-1.222,-0.547,1.138", Rotation="0,0,0", Scale="1,1,1") -> (Pose) {
    alphaInputType=Bool
    rotationMode=BMM_Additive
    rotationSpace=BCS_BoneSpace
    translationMode=BMM_Additive
    translationSpace=BCS_ParentBoneSpace
    alphaBoolBlend {blendInTime=0.2, blendOption=Linear, blendOutTime=0.2}
    boneToModify {boneName=dyn_chain_mid_03_joint2}
    translation=<-1.222 -0.547 1.138>
   }
   n49 = Get isNotMoving() -> (isNotMoving:double)
   n50 = "Constraint - dyn_chain_mid_03_joint (PPPPPP)"!(
    ComponentPose=@n48
    Alpha=@n49
    ConstraintWeights_0=0.6927
    ConstraintWeights_1:float=0.0589
    ConstraintWeights_2:float=0.1462
    ConstraintWeights_3:float=0.0208
    ConstraintWeights_4:float=0.0772
    ConstraintWeights_5:float=0.0042
   ) -> (Pose) {
    alphaBoolBlend {blendOption=Linear}
    alphaScaleBiasClamp {bInterpResult=true, scale=0.8}
    boneToModify {boneName=dyn_chain_mid_03_joint}
    constraintSetup [
     {targetBone {boneName=dyn_skirt_side_a_l}}
     {targetBone {boneName=dyn_skirt_side_b_l}}
     {targetBone {boneName=pelvis}}
     {targetBone {boneName=dyn_skirt_bk_a_l}}
     {targetBone {boneName=dyn_skirt_fr_a_l}}
     {targetBone {boneName=dyn_skirt_fr_b_l}}
    ]
    constraintWeights [1,1,1,1,1,1]
   }
   n51 = "Constraint - dyn_chain_front_01_joint (PPP)"(ComponentPose=@n50, Alpha=@n49, ConstraintWeights_0=0.5539, ConstraintWeights_1=0.442, ConstraintWeights_2=0.004) {
    alphaBoolBlend {blendOption=Linear}
    alphaScaleBiasClamp {bInterpResult=true, scale=0.45}
    boneToModify {boneName=dyn_chain_front_01_joint}
    constraintSetup [{targetBone {boneName=pelvis}}, {targetBone {boneName=dyn_skirt_fr_a_l}}, {targetBone {boneName=spine_01}}]
    constraintWeights [1,1,1]
   }
   n52 = Get No_Idel() -> (No_Idel:double)
   n53 = "Transform (Modify) Bone - Bone: dyn_strap"(ComponentPose=@n51, Alpha=@n52, Translation="0,-0.2,0", Rotation="-0.268,20.267,11.422", Scale="1,1,1") {
    rotationMode=BMM_Additive
    rotationSpace=BCS_BoneSpace
    translationMode=BMM_Additive
    translationSpace=BCS_BoneSpace
    alphaBoolBlend {blendOption=Linear}
    alphaScaleBiasClamp {bInterpResult=true}
    boneToModify {boneName=dyn_strap}
    rotation=<rot -0.268 20.267 11.422>
    translation {y=-0.2}
   }
   n54 = "Transform (Modify) Bone - Bone: dyn_chain_front_01_joint"!(ComponentPose=@n53, Alpha:float=@n52, Translation="-0.001,-0.094,0.709", Rotation="-13.480,-18.157,0.780", Scale="1,1,1") -> (Pose) {
    rotationMode=BMM_Additive
    rotationSpace=BCS_BoneSpace
    translationMode=BMM_Additive
    translationSpace=BCS_BoneSpace
    alphaBoolBlend {blendOption=Linear}
    alphaScaleBiasClamp {bInterpResult=true}
    boneToModify {boneName=dyn_chain_front_01_joint}
    rotation=<rot -13.48 -18.157 0.78>
    translation=<-0.001 -0.094 0.709>
   }
   n55 = "Transform (Modify) Bone - Bone: dyn_chain_mid_03_joint2"!(ComponentPose=@n54, Alpha:float=@n52, Translation="0.132,0.842,0.406", Rotation="-1.426,0.052,6.974", Scale="1,1,1") -> (Pose) {
    rotationMode=BMM_Additive
    rotationSpace=BCS_BoneSpace
    translationMode=BMM_Additive
    translationSpace=BCS_BoneSpace
    alphaBoolBlend {blendOption=Linear}
    alphaScaleBiasClamp {bInterpResult=true}
    boneToModify {boneName=dyn_chain_mid_03_joint2}
    rotation=<rot -1.426 0.052 6.974>
    translation=<0.132 0.842 0.406>
   }
   n56 = "Transform (Modify) Bone - Bone: dyn_chain_mid_03_joint"(ComponentPose=@n55, Alpha=@n52, Translation="-1.177,-1.583,-4.141", Rotation="3,-0.086,-0.063", Scale="1,1,1") {
    rotationMode=BMM_Additive
    rotationSpace=BCS_BoneSpace
    translationMode=BMM_Additive
    translationSpace=BCS_BoneSpace
    alphaBoolBlend {blendOption=Linear}
    alphaScaleBiasClamp {bInterpResult=true}
    boneToModify {boneName=dyn_chain_mid_03_joint}
    rotation=<rot 3 -0.086 -0.063>
    translation=<-1.177 -1.583 -4.141>
   }
   n57 = "Transform (Modify) Bone - Bone: dyn_chain_mid_03_joint3"(ComponentPose=@n56, Alpha=@n52, Translation="0.906,-0.387,3.485", Rotation="6.476,0.000,0.000", Scale="1,1,1") {
    rotationMode=BMM_Additive
    rotationSpace=BCS_BoneSpace
    translationMode=BMM_Additive
    translationSpace=BCS_BoneSpace
    alphaBoolBlend {blendOption=Linear}
    alphaScaleBiasClamp {bInterpResult=true}
    boneToModify {boneName=dyn_chain_mid_03_joint3}
    rotation=<rot 6.476 0 0>
    translation=<0.906 -0.387 3.485>
   }
   n58 = "Transform (Modify) Bone - Bone: dyn_chain_front_01_joint"(
    ComponentPose=@n57
    Alpha=Get is_Rigid_Body() -> (is_Rigid_Body:double)
    Translation="-1.669,-1.316,-0.550"
    Rotation="-13.279,-25.428,2.464"
    Scale="1,1,1"
   ) {
    rotationMode=BMM_Additive
    rotationSpace=BCS_BoneSpace
    translationMode=BMM_Additive
    translationSpace=BCS_BoneSpace
    alphaBoolBlend {blendOption=Linear}
    alphaScaleBiasClamp {bInterpResult=true}
    boneToModify {boneName=dyn_chain_front_01_joint}
    rotation=<rot -13.279 -25.428 2.464>
    translation=<-1.669 -1.316 -0.55>
   }
   Output Pose(Result=Component To Local(ComponentPose=@n58)) {layerGroup=None}
   Reroute Node(InputPin:bool=Get bIsCrouching()) -> (OutputPin:bool)
   Get Left Move() -> (Left Move:bool)
   Get Right Move() -> (Right Move:bool)
   Get bIsCrouchSprinting()
  ]}
 EventGraph {nodes [
   n1 = Event BlueprintUpdateAnimation()
   > n2 = Sequence!(execute) -> (then_0, then_1, then_2:Exec, then_3:Exec)
   n3 = Sequence!(execute=@n2.then_3) -> (then_0, then_1, then_2, then_3, then_4:Exec, then_5:Exec, then_6:Exec, then_7:Exec)
   n4 = NOT Boolean(A=Get bIsPlayingEmote())
   n5 = Get bRigidBodyEnabled()
   n6 = Get bIsInFrontEnd()
   Set No_Idel(execute=@n3.then_6, No_Idel:double="To Float (Boolean)"(InBool=AND Boolean!(A=@n6, B=@n4, C:bool=@n5) -> (ReturnValue))) -> (Output_Get:double)
   n7 = Get bRigidBodyEnabled()
   > Set is_Rigid_Body(is_Rigid_Body:double="To Float (Boolean)"(InBool=NOT Boolean(A=@n7))) -> (Output_Get:double)
   n8 = Enum to String!(Enumerator:EFortCardinalDirection Enum=GetCardinalDirectionFromAngle!(
     self:Fort Anim Instance Object Reference
     Angle:float=Get PawnMovementDirectionAzimuth() -> (PawnMovementDirectionAzimuth:float)
     DeadZoneAngle:float=0
    ) -> (ReturnValue:EFortCardinalDirection Enum=North)) -> (ReturnValue:String)
   n9 = "Equal Exactly (String)"(A=@n8, B=West)
   n10 = Branch(execute=@n3.then_2, Condition=@n9)
   > Set chain left(chain left:bool=true) -> (Output_Get:bool)
   Set chain left(execute=@n10.else, chain left=false)
   n11 = "To Float (Boolean)"(InBool=Get bIsMoving2D())
   n12 = Get bIsCrouching()
   Set Con chain crounch(
    execute=@n3.then_1
    Con chain crounch:double=MapRangeClamped(
     Value="To Float (Boolean)"(InBool=@n12)
     InRangeA=0
     InRangeB=1
     OutRangeB=MapRangeClamped(Value=@n11, InRangeA=0, InRangeB=1, OutRangeA=0.3, OutRangeB=0.5)
    )
   ) -> (Output_Get:double)
   n13 = "To Float (Boolean)"(InBool=Get bIsSkyDiving())
   n14 = "To Float (Boolean)"(InBool=Get bIsCrouching())
   Set Con Chain(
    execute=@n3.then_4
    Con Chain:double=MapRangeClamped(Value=@n14, InRangeA=0, InRangeB=1, OutRangeA=MapRangeClamped(Value=@n13, InRangeA=0, InRangeB=1))
   ) -> (Output_Get:double)
   n15 = MapRangeClamped(Value=@n14, InRangeA=0, InRangeB=1, OutRangeA=MapRangeClamped(Value=@n13, InRangeA=0, InRangeB=1, OutRangeA=1, OutRangeB=0), OutRangeB=0)
   > Set Dyn Chain(Dyn Chain:double=@n15) -> (Output_Get:double)
   > Set Dyn Skirt_inv(Dyn Skirt_inv:double=MapRangeClamped(Value=@n15, InRangeA=0, InRangeB=1, OutRangeA=1, OutRangeB=0)) -> (Output_Get:double)
   > Set Dyn Skirt(Dyn Skirt:double=MapRangeClamped(Value=@n14, InRangeA=0, InRangeB=1, OutRangeA=MapRangeClamped(Value=@n13, InRangeA=0, InRangeB=1, OutRangeA=1, OutRangeB=0.3), OutRangeB=0.5)) -> (Output_Get:double)
   n16 = Branch(execute=@n3.then_0, Condition=AND Boolean(A=Get bIsCrouching(), B=Get bIsMoving2D(), C=Get bIsBackpedaling() -> (bIsBackpedaling:bool)))
   > Set back crouch strap(back crouch strap:bool=true) -> (Output_Get:bool)
   Set back crouch strap(execute=@n16.else, back crouch strap=false)
   n17 = Sequence!(execute=@n3.then_3) -> (then_0, then_1)
   > n18 = Branch(Condition=AND Boolean!(A=Get bIsCrouching(), B=Get bIsMoving2D()) -> (ReturnValue))
   > n19 = Set Left Move(Left Move:bool=@n9) -> (Output_Get:bool)
   > Set Left MoveFloat(Left MoveFloat:double="To Float (Boolean)"(InBool=@n19)) -> (Output_Get:double)
   n20 = "Equal Exactly (String)"(A=@n8, B=East)
   > n21 = Set Right Move(Right Move:bool=@n20) -> (Output_Get:bool)
   > Set Right MoveFloat(Right MoveFloat:double="To Float (Boolean)"(InBool=@n21)) -> (Output_Get:double)
   > n22 = Set Crouching Move(Crouching Move:bool=OR Boolean!(A=Get bIsMoving2D(), B=OR Boolean!(A=false, B=@n9) -> (ReturnValue), C:bool=@n20) -> (ReturnValue)) -> (Output_Get:bool)
   > "Set Crouching Move Float "("Crouching Move Float ":float="To Float (Boolean)"(InBool=@n22)) -> (Output_Get:float)
   n23 = Set Left Move(execute=@n18.else, Left Move=false)
   > Set Left MoveFloat(Left MoveFloat="To Float (Boolean)"(InBool=@n23))
   > n24 = Set Right Move(Right Move=false)
   > Set Right MoveFloat(Right MoveFloat="To Float (Boolean)"(InBool=@n24))
   > n25 = Set Crouching Move(Crouching Move=false)
   > "Set Crouching Move Float "("Crouching Move Float "="To Float (Boolean)"(InBool=@n25))
   Set R leg skydive fix(execute=@n17.then_1, R leg skydive fix:bool=OR Boolean(A=Get bIsCrouching(), B=Get bIsParachuteOpened())) -> (Output_Get:bool)
   n26 = Get bIsMoving2D()
   Set isNotMoving(execute=@n3.then_5, isNotMoving:double="To Float (Boolean)"(InBool=NOT Boolean(A=@n26))) -> (Output_Get:double)
   > Set isCrouchingFloat(isCrouchingFloat:double="To Float (Boolean)"(InBool=Get bIsCrouching())) -> (Output_Get:double)
   > Set isCrouchSprinting(isCrouchSprinting:float="To Float (Boolean)"(InBool=Get bIsCrouchSprinting())) -> (Output_Get:float)
   n27 = Gravity Override State Function DEPRECATED(
    execute=@n2.then_0
    CharacterPartAnimInstance=Self-Reference()
    GameTimeInSeconds=GetGameTimeInSeconds()
    GravityOverrideParams=Get Gravity Override Params()
   )
   > Set Gravity(Gravity:Vector=@n27) -> (Output_Get:Vector)
   n28 = GetOwnerSkeletalMeshComponent()
   Set Con End Chain(
    execute=@n2.then_1
    Con End Chain:double=MapRangeClamped(
     Value=GetAlphaFromReferenceTransformDelta(SkeletalMeshComponent=@n28, BoneName=head, BaseName=spine_05, InputRange=(X=-10,Y=-48), TwistAxis="1,0,0")
     InRangeA=0
     InRangeB=0.6
     OutRangeA=1
    )
   ) -> (Output_Get:double)
   Set thigh_l_rot_alpha(
    execute=@n2.then_2
    thigh_l_rot_alpha:double=GetAlphaFromReferenceTransformDelta(SkeletalMeshComponent=@n28, BoneName=thigh_l, BaseName=pelvis, InputRange=(X=-19,Y=-100), TwistAxis="1,0,0")
   ) -> (Output_Get:double)
   > n29 = Branch(Condition=Get bIsJumping())
   > Set thigh_r_rot_alpha(thigh_r_rot_alpha:double=GetAlphaFromReferenceTransformDelta(SkeletalMeshComponent=@n28, BoneName=thigh_r, BaseName=pelvis, InputRange=(X=-10,Y=-70), TwistAxis="1,0,0")) -> (Output_Get:double)
   Set thigh_r_rot_alpha(execute=@n29.else, thigh_r_rot_alpha=FInterpTo(Current=Get thigh_r_rot_alpha(), Target=0, DeltaTime=@n1.DeltaTimeX, InterpSpeed=1))
  ]}
]
variables [
 back crouch strap {type=boolean}
 chain left {type=boolean}
 Con Chain {type=number}
 Con chain crounch {type=number}
 Con End Chain {type=number}
 Crouching Move {type=boolean}
 "Crouching Move Float " {type=number}
 Dyn Chain {type=number}
 Dyn Skirt {type=number}
 Dyn Skirt_inv {type=number}
 Gravity {type=Vector}
 Gravity Override Params {type=GravityOverrideParamsStruct}
 is_Rigid_Body {type=number}
 isCrouchingFloat {type=number}
 isCrouchSprinting {type=number}
 isNotMoving {type=number}
 Left Move {type=boolean}
 Left MoveFloat {type=number}
 No_Idel {type=number}
 R leg skydive fix {type=boolean}
 Right Move {type=boolean}
 Right MoveFloat {type=number}
 switch_dyn_chain {type=number}
 thigh_l_rot_alpha {type=number}
 thigh_r_rot_alpha {type=number}
]
