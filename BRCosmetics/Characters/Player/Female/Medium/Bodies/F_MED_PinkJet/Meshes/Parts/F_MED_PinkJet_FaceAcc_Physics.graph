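# Physics asset for the F_MED_PinkJet face accessory mesh, covering dynamic hair and barrette bones on this female player cosmetic.
# Capsule bodies cover left and right barrettes, back hair, two bang chains, and the head and neck.
# Barrettes and back hair bones each constrain directly to the head, while each bang chains root to middle to end from it.
# Constraints limit swing, tightest at bang roots, and nearly lock twist.
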PhysicsAsset /BRCosmetics/Characters/Player/Female/Medium/Bodies/F_MED_PinkJet/Meshes/Parts/F_MED_PinkJet_FaceAcc_Physics
bodies {
 dyn_L_barrette_down_a [{
   shapeType=Capsule
   length=0.505
   radius=0.505
   shapeName=dyn_L_barrette_down_a_capsule
   center=<1.2 0.0341 0.0675>
   rotation=<rot 3.024 -90 -5.1753>
  }]
 dyn_L_barrette_down_b [{
   shapeType=Capsule
   length=0.505
   radius=0.505
   shapeName=dyn_L_barrette_down_b_capsule
   center=<1.4718 0.0488 -0.0062>
   rotation=<rot 15.3277 -90 -1.8991>
  }]
 dyn_L_barrette_up_a [{
   shapeType=Capsule
   length=0.505
   radius=0.505
   shapeName=dyn_L_barrette_up_a_capsule
   center=<0.7711 -0.0918 0.0347>
   rotation=<rot -13.2858 -90 -8.4601>
  }]
 dyn_L_barrette_up_b [{
   shapeType=Capsule
   length=0.505
   radius=0.505
   shapeName=dyn_L_barrette_up_b_capsule
   center=<0.5329 -0.2502 0.0094>
   rotation=<rot 8.7545 -90 -2.308>
  }]
 dyn_R_barrette_down_a [{
   shapeType=Capsule
   length=0.505
   radius=0.505
   shapeName=dyn_R_barrette_down_a_capsule
   center=<0.0234 -1.7285 -0.1011>
   rotation=<rot -0.8625 -87.362 -9.7452>
  }]
 dyn_R_barrette_down_b [{
   shapeType=Capsule
   length=0.505
   radius=0.505
   shapeName=dyn_R_barrette_down_b_capsule
   center=<0.0516 -1.9166 0.0119>
   rotation=<rot -4.3646 -89.9981 -3.2341>
  }]
 dyn_R_barrette_up_a [{
   shapeType=Capsule
   length=0.505
   radius=0.505
   shapeName=dyn_R_barrette_up_a_capsule
   center=<-0.1736 -0.8035 0.0211>
   rotation=<rot -9.4545 -89.2414 18.4595>
  }]
 dyn_R_barrette_up_b [{
   shapeType=Capsule
   length=0.7885
   radius=0.5
   shapeName=dyn_R_barrette_up_b_capsule
   center=<-0.0226 -1.0496 0.0011>
   rotation=<rot -2.6881 -90.291 -13.3654>
  }]
 dyn_hair_back_a [{shapeType=Capsule, length=0.6758, radius=0.7991, shapeName=dyn_hair_back_a_capsule, center=<0.9219 0 0>, rotation=<rot 0 -90 0>}]
 dyn_hair_back_b [{shapeType=Capsule, length=0.6758, radius=0.7991, shapeName=dyn_hair_back_b_capsule, center=<1.2506 0 0>, rotation=<rot 0 -90 0>}]
 dyn_hair_back_c [{shapeType=Capsule, length=0.6758, radius=0.7991, shapeName=dyn_hair_back_c_capsule, center=<1.3051 0 0>, rotation=<rot 0 -90 0>}]
 dyn_hair_bang_a [{shapeType=Capsule, length=0.505, radius=0.505, shapeName=dyn_hair_bang_a_capsule, rotation=<rot 0 -90 0>}]
 dyn_hair_bang_a_end [{shapeType=Capsule, length=0.505, radius=0.505, shapeName=dyn_hair_bang_a_end_capsule, rotation=<rot 0 -90 0>}]
 dyn_hair_bang_b [{shapeType=Capsule, length=0.505, radius=0.505, shapeName=dyn_hair_bang_b_capsule, rotation=<rot 0 -90 0>}]
 dyn_hair_bang_b_end [{shapeType=Capsule, length=0.505, radius=0.505, shapeName=dyn_hair_bang_b_end_capsule, rotation=<rot 0 -90 0>}]
 dyn_hair_root_bang_a [{shapeType=Capsule, length=0.505, radius=0.505, shapeName=dyn_hair_root_bang_a_capsule, center=<1.9574 0 0>, rotation=<rot 0 -90 0>}]
 dyn_hair_root_bang_b [{shapeType=Capsule, length=0.8175, radius=0.5, shapeName=dyn_hair_root_bang_b_capsule, center=<1.3927 0 0>, rotation=<rot 0 -90 0>}]
 head [
  {
   shapeType=Capsule
   length=8.7188
   radius=4.5721
   shapeName=head_capsule
   center=<5.5015 0.0621 0.2878>
   rotation=<rot 7.6662 -89.4683 1.2253>
  }
  {shapeType=Capsule, length=6.038, radius=3.689, shapeName=head_capsule, center=<0 -4.8255 0>}
 ]
 neck_02 [{shapeType=Capsule, length=9.401, radius=3.8943, shapeName=neck_02_capsule, center=<-0.472 -4.785 0>}]
}
constraints [
 {bone1Name=dyn_L_barrette_down_a, bone2Name=head, swing1LimitDegrees=30, swing2LimitDegrees=30, twistLimitDegrees=0.01}
 {bone1Name=dyn_L_barrette_down_b, bone2Name=head, swing1LimitDegrees=30, swing2LimitDegrees=30, twistLimitDegrees=0.01}
 {bone1Name=dyn_L_barrette_up_a, bone2Name=head, swing1LimitDegrees=30, swing2LimitDegrees=30, twistLimitDegrees=0.01}
 {bone1Name=dyn_L_barrette_up_b, bone2Name=head, swing1LimitDegrees=30, swing2LimitDegrees=30, twistLimitDegrees=0.01}
 {bone1Name=dyn_R_barrette_down_a, bone2Name=head, swing1LimitDegrees=30, swing2LimitDegrees=30, twistLimitDegrees=0.01}
 {bone1Name=dyn_R_barrette_down_b, bone2Name=head, swing1LimitDegrees=30, swing2LimitDegrees=30, twistLimitDegrees=0.01}
 {bone1Name=dyn_R_barrette_up_a, bone2Name=head, swing1LimitDegrees=30, swing2LimitDegrees=30, twistLimitDegrees=0.01}
 {bone1Name=dyn_R_barrette_up_b, bone2Name=head, swing1LimitDegrees=30, swing2LimitDegrees=30, twistLimitDegrees=0.01}
 {bone1Name=dyn_hair_back_a, bone2Name=head, swing1LimitDegrees=30, swing2LimitDegrees=15, twistLimitDegrees=0.01}
 {bone1Name=dyn_hair_back_b, bone2Name=head, swing1LimitDegrees=30, swing2LimitDegrees=15, twistLimitDegrees=0.01}
 {bone1Name=dyn_hair_back_c, bone2Name=head, swing1LimitDegrees=30, swing2LimitDegrees=15, twistLimitDegrees=0.01}
 {bone1Name=dyn_hair_bang_a, bone2Name=dyn_hair_root_bang_a, swing1LimitDegrees=30, swing2LimitDegrees=15, twistLimitDegrees=0.01}
 {bone1Name=dyn_hair_bang_a_end, bone2Name=dyn_hair_bang_a, swing1LimitDegrees=30, swing2LimitDegrees=15, twistLimitDegrees=0.01}
 {bone1Name=dyn_hair_bang_b, bone2Name=dyn_hair_root_bang_b, swing1LimitDegrees=30, swing2LimitDegrees=15, twistLimitDegrees=0.01}
 {bone1Name=dyn_hair_bang_b_end, bone2Name=dyn_hair_bang_b, swing1LimitDegrees=30, swing2LimitDegrees=15, twistLimitDegrees=0.01}
 {bone1Name=dyn_hair_root_bang_a, bone2Name=head, swing1LimitDegrees=5, swing2LimitDegrees=10, twistLimitDegrees=0.01}
 {bone1Name=dyn_hair_root_bang_b, bone2Name=head, swing1LimitDegrees=5, swing2LimitDegrees=10, twistLimitDegrees=0.01}
]
