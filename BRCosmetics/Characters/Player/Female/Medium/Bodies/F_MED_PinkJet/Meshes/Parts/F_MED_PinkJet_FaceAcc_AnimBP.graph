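# Animation Blueprint for the PinkJet female face accessory, driving three dynamic back-hair bones.
# It copies the parent mesh pose, simulates RigidBody physics in neck_02 space with overridable gravity, then constrains the hair to head, neck and spine bones.
# Each update, head rotation relative to spine_05 inversely trades physics alpha against constraint alpha.
# The jump and aim bone-transform chains and the weapon-targeting logic are disconnected.
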AnimBlueprint /BRCosmetics/Characters/Player/Female/Medium/Bodies/F_MED_PinkJet/Meshes/Parts/F_MED_PinkJet_FaceAcc_AnimBP
target_skeleton=&/F_MED_PinkJet_FaceAcc_Skeleton
graphs [
 AnimGraph {nodes [
   n1 = Get bIsJumping()
   n2 = "Transform (Modify) Bone - Bone: dyn_hair_back_a"!(ComponentPose, bAlphaBoolEnabled:bool=@n1, Translation="0,0,0", Rotation="50,-0.002,-0.004", Scale="1,1,1") -> (Pose) {
    alphaInputType=Bool
    rotationMode=BMM_Additive
    rotationSpace=BCS_BoneSpace
    alphaBoolBlend {blendOption=Linear, blendOutTime=2}
    boneToModify {boneName=dyn_hair_back_a}
    rotation=<rot 50 -0.002 -0.004>
   }
   n3 = "Transform (Modify) Bone - Bone: dyn_hair_back_b"!(ComponentPose=@n2, bAlphaBoolEnabled:bool=@n1, Translation="0,0,0", Rotation="50,-0.002,-0.004", Scale="1,1,1") -> (Pose) {
    alphaInputType=Bool
    rotationMode=BMM_Additive
    rotationSpace=BCS_BoneSpace
    alphaBoolBlend {blendOption=Linear, blendOutTime=2}
    boneToModify {boneName=dyn_hair_back_b}
    rotation=<rot 50 -0.002 -0.004>
   }
   "Transform (Modify) Bone - Bone: dyn_hair_back_c"!(ComponentPose=@n3, bAlphaBoolEnabled:bool=@n1, Translation="0,0,0", Rotation="50,-0.002,-0.004", Scale="1,1,1") -> (Pose) {
    alphaInputType=Bool
    rotationMode=BMM_Additive
    rotationSpace=BCS_BoneSpace
    alphaBoolBlend {blendOption=Linear, blendOutTime=2}
    boneToModify {boneName=dyn_hair_back_c}
    rotation=<rot 50 -0.002 -0.004>
   }
   "Transform (Modify) Bone - Bone: dyn_hair_back_c"(bAlphaBoolEnabled=True, Translation="0,0,0", Rotation="50,-0.002,-0.004", Scale="1,1,1") {
    alphaInputType=Bool
    rotationMode=BMM_Additive
    rotationSpace=BCS_BoneSpace
    alphaBoolBlend {blendOption=Linear, blendOutTime=2}
    boneToModify {boneName=dyn_hair_back_c}
    rotation=<rot 50 -0.002 -0.004>
   }
   n4 = Get aim() -> (aim:double)
   n5 = "Transform (Modify) Bone - Bone: dyn_hair_back_a"!(ComponentPose, Alpha:float=@n4, Translation="0,0,3", Rotation="10,0.003,-0.003", Scale="0.2,0.2,0.2") -> (Pose) {
    rotationMode=BMM_Additive
    rotationSpace=BCS_BoneSpace
    scaleMode=BMM_Additive
    scaleSpace=BCS_BoneSpace
    translationMode=BMM_Additive
    translationSpace=BCS_BoneSpace
    alphaBoolBlend {blendOption=Linear}
    boneToModify {boneName=dyn_hair_back_a}
    rotation=<rot 10 0.003 -0.003>
    scale=<0.2 0.2 0.2>
    translation {z=3}
   }
   n6 = "Transform (Modify) Bone - Bone: dyn_hair_back_b"!(ComponentPose=@n5, Alpha:float=@n4, Translation="-1.147,2.074,7.107", Rotation="10,0,0", Scale="0.2,0.2,0.2") -> (Pose) {
    rotationMode=BMM_Additive
    rotationSpace=BCS_BoneSpace
    scaleMode=BMM_Additive
    scaleSpace=BCS_BoneSpace
    translationMode=BMM_Additive
    translationSpace=BCS_BoneSpace
    alphaBoolBlend {blendOption=Linear}
    boneToModify {boneName=dyn_hair_back_b}
    rotation {pitch=10}
    scale=<0.2 0.2 0.2>
    translation=<-1.147 2.074 7.107>
   }
   "Transform (Modify) Bone - Bone: dyn_hair_back_c"!(ComponentPose=@n6, Alpha:float=@n4, Translation="6.475,6.647,11.748", Rotation="0,0,-10", Scale="0.2,0.2,0.2") -> (Pose) {
    rotationMode=BMM_Additive
    rotationSpace=BCS_BoneSpace
    scaleMode=BMM_Additive
    scaleSpace=BCS_BoneSpace
    translationMode=BMM_Additive
    translationSpace=BCS_BoneSpace
    alphaBoolBlend {blendOption=Linear}
    boneToModify {boneName=dyn_hair_back_c}
    rotation {roll=-10}
    scale=<0.2 0.2 0.2>
    translation=<6.475 6.647 11.748>
   }
   n7 = RigidBody!(
    ComponentPose=Local To Component(LocalPose=Copy Pose From Mesh() {bUseAttachedParent=true})
    Alpha=Get invert_amp() -> (invert_amp:double)
    OverrideWorldGravity:Vector=Get Gravity() -> (Gravity:Vector)
    ExternalForce
   ) -> (Pose) {
    bClampLinearTranslationLimitToRefPose=true
    bOverrideWorldGravity=true
    overridePhysicsAsset=&/F_MED_PinkJet_FaceAcc_Physics
    simulationSpace=BaseBoneSpace
    alphaBoolBlend {blendOption=Linear}
    baseBoneRef {boneName=neck_02}
    componentAppliedLinearAccClamp=<3000 3000 3000>
    componentLinearAccScale=<0.5 0.5 0.5>
    componentLinearVelScale=<0.5 0.5 0.5>
    overrideWorldGravity {z=-980}
    simSpaceSettings {maxAngularAcceleration=3000, maxAngularVelocity=3000, maxLinearAcceleration=3000, maxLinearVelocity=3000}
   }
   n8 = Get aim()
   n9 = "Constraint - dyn_hair_back_a (PPPP)"!(
    ComponentPose=@n7
    Alpha=@n8
    ConstraintWeights_0=0.0854
    ConstraintWeights_1:float=0.6151
    ConstraintWeights_2:float=0.1888
    ConstraintWeights_3:float=0.1108
   ) -> (Pose) {
    alphaBoolBlend {blendOption=Linear}
    boneToModify {boneName=dyn_hair_back_a}
    constraintSetup [{targetBone {boneName=head}}, {targetBone {boneName=spine_05}}, {targetBone {boneName=neck_02}}, {targetBone {boneName=neck_01}}]
    constraintWeights [1,1,1,1]
   }
   n10 = "Constraint - dyn_hair_back_b (PPPP)"!(
    ComponentPose=@n9
    Alpha=@n8
    ConstraintWeights_0=0.1178
    ConstraintWeights_1:float=0.377
    ConstraintWeights_2:float=0.3362
    ConstraintWeights_3:float=0.169
   ) -> (Pose) {
    alphaBoolBlend {blendOption=Linear}
    boneToModify {boneName=dyn_hair_back_b}
    constraintSetup [{targetBone {boneName=head}}, {targetBone {boneName=spine_05}}, {targetBone {boneName=neck_02}}, {targetBone {boneName=neck_01}}]
    constraintWeights [1,1,1,1]
   }
   n11 = "Constraint - dyn_hair_back_c (PPPP)"!(
    ComponentPose=@n10
    Alpha=@n8
    ConstraintWeights_0=0.0787
    ConstraintWeights_1:float=0.6112
    ConstraintWeights_2:float=0.1951
    ConstraintWeights_3:float=0.115
   ) -> (Pose) {
    alphaBoolBlend {blendOption=Linear}
    boneToModify {boneName=dyn_hair_back_c}
    constraintSetup [{targetBone {boneName=head}}, {targetBone {boneName=spine_05}}, {targetBone {boneName=neck_02}}, {targetBone {boneName=neck_01}}]
    constraintWeights [1,1,1,1]
   }
   Output Pose(Result=Component To Local(ComponentPose=@n11)) {layerGroup=None}
  ]}
 EventGraph {nodes [
   n1 = Reroute Node(InputPin:Skeletal Mesh Component Object Reference) -> (OutputPin:Skeletal Mesh Component Object Reference)
   n2 = Cast To FortPlayerAnimInstance(Object=GetAnimInstance(self=@n1))
   > n3 = Branch(Condition=Get bIsTargeting())
   > n4 = Branch(Condition="Equal (Enum)"(A:EFortWeaponCoreAnimation Enum=Get WeaponCoreAnim(self=@n2), B:EFortWeaponCoreAnimation Enum=Pistol))
   n5 = Get bIsCrouching()
   > n6 = Branch(Condition=@n5)
   n7 = GetTwistAndSwingAngleOfDeltaRotationFromRefPose(self=@n1, BoneName=upperarm_r)
   > Set aim(aim:double=MapRangeClamped(Value=@n7.OutSwingAngle, InRangeA=110, InRangeB=135, OutRangeB=0.3)) -> (Output_Get:double)
   Set aim(execute=@n6.else, aim=MapRangeClamped(Value=@n7.OutSwingAngle, InRangeA=110, InRangeB=135, OutRangeB=0.5))
   n8 = Branch(execute=@n4.else, Condition=@n5)
   > Set aim(aim=MapRangeClamped(Value=@n7.OutSwingAngle, InRangeA=80, InRangeB=88, OutRangeB=0.3))
   Set aim(execute=@n8.else, aim=MapRangeClamped(Value=@n7.OutSwingAngle, InRangeA=80, InRangeB=88, OutRangeB=0.3))
   Set aim(execute=@n3.else, aim=0)
   Event BlueprintUpdateAnimation()
   > n9 = Sequence!(execute) -> (then_0, then_1, then_2:Exec)
   n10 = Gravity Override State Function DEPRECATED(
    execute=@n9.then_1
    CharacterPartAnimInstance=Self-Reference()
    GameTimeInSeconds=GetGameTimeInSeconds()
    GravityOverrideParams=Get Gravity Override Params()
   )
   > Set Gravity(Gravity:Vector=@n10) -> (Output_Get:Vector)
   > Set Is Jumping(Is Jumping:bool=Get bIsJumping()) -> (Output_Get:bool)
   n11 = MapRangeClamped(
    Value=GetAlphaFromReferenceTransformDelta(SkeletalMeshComponent=GetOwnerSkeletalMeshComponent(), BoneName=head, BaseName=spine_05, InputRange=(X=-20,Y=10), TwistAxis="1,0,0")
    InRangeA=0.2
    InRangeB=0.6
    OutRangeA=1
    OutRangeB=0
   )
   Set invert_amp(execute=@n9.then_0, invert_amp:double=MapRangeClamped(Value=@n11, InRangeA=1, InRangeB=0, OutRangeB=0.6)) -> (Output_Get:double)
   > Set aim(aim=MapRangeClamped(Value=@n11, InRangeA=0, InRangeB=1))
   MapRangeClamped(Value=@n7.OutSwingAngle, InRangeA=110, InRangeB=135, OutRangeB=0.3)
  ]}
]
variables [
 aim {type=number}
 Gravity {type=Vector}
 Gravity Override Params {type=GravityOverrideParamsStruct}
 invert_amp {type=number}
 Is Jumping {type=boolean}
]
